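# Conversation database for the RaysNight NPC, who introduces themself as Khari and talks with the player.
# A single entry point leads to a speech node that picks a contextual greeting by requirement and priority: first-ever meeting, reactions to two specific outfits, or lines gated on match conversation count.
# The speech node then links out to the shared NPC library Quests and BasicServices conversation entries.
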ConversationDatabase /NPCLibrary/NPCs/RaysNight/NPCLibrary_Conversation_RaysNight
entry_points [{tag=Conversation.Entry.RaysNight, destinations [#1]}]
nodes [
 ConversationEntryPointNode_7 {
  id=#1
  class=ConversationEntryPointNode
  outputs [#2]
  properties {entryTag {tagName=Conversation.Entry.RaysNight}}
 }
 Speech_C_4 {
  id=#2
  class=Speech_C
  outputs [#3, #4]
  properties {
   nodeName=""
   generalConfig {
    defaultMessage=Trust me. I never miss.
    contextualMessages [
     {
      message=Nice to meet you. I'm Khari.
      contextRequirements [{requirement {class=FortControllerRequirement_IsFirstEverConversationWithNPC}}]
      priority {curve {rowName=Tandem.Speech.Priority.FirstEverInteraction}}
     }
     {
      message="Oh. Uhhh. Hi! It's... It's a wonder to meet you."
      contextRequirements [{requirement {
         class=FortControllerRequirement_HasCID
         properties {allowedCIds [/BRCosmetics/Athena/Items/Cosmetics/Characters/Character_SeleneCobra_Stride]}
        }}]
      priority {curve {rowName=Tandem.Speech.Priority.OutfitReaction}}
     }
     {
      message="The gods aren't all bad... I mean, Artemis seems pretty cool..."
      contextRequirements [{requirement {
         class=FortControllerRequirement_HasCID
         properties {allowedCIds [/BRCosmetics/Athena/Items/Cosmetics/Characters/CID_A_332_Athena_Commando_F_PrimalFalcon_3ITKM]}
        }}]
      priority {curve {rowName=Tandem.Speech.Priority.OutfitReaction}}
     }
     {
      message=Trust me. I never miss.
      contextRequirements [{requirement {class=FortControllerRequirement_CheckNPCMatchConversationCount, properties {nPCUniqueIdTag {tagName=AISpawnerData.Type.Tandem.RaysNight}}}}]
      priority {curve {rowName=None}}
     }
     {
      message=If Artemis called... I could be persuaded to join her.
      contextRequirements [{requirement {class=FortControllerRequirement_CheckNPCMatchConversationCount, properties {nPCUniqueIdTag {tagName=AISpawnerData.Type.Tandem.RaysNight}}}}]
      priority {curve {rowName=None}}
     }
     {
      message=Have you ever held destiny in the palm of your hand?
      contextRequirements [{requirement {class=FortControllerRequirement_CheckNPCMatchConversationCount, properties {nPCUniqueIdTag {tagName=AISpawnerData.Type.Tandem.RaysNight}}}}]
      priority {curve {rowName=None}}
     }
     {
      message=I'll never trust other weapons like I do my own bow.
      contextRequirements [{requirement {class=FortControllerRequirement_CheckNPCMatchConversationCount, properties {nPCUniqueIdTag {tagName=AISpawnerData.Type.Tandem.RaysNight}}}}]
      priority {curve {rowName=None}}
     }
    ]
   }
  }
 }
 ConversationLinkNode_0 {id=#4, class=ConversationLinkNode, properties {remoteEntryTag {tagName=Conversation.Entry.BasicServices}}}
 ConversationLinkNode_1 {id=#3, class=ConversationLinkNode, properties {remoteEntryTag {tagName=Conversation.Entry.NPCLibrary.Quests}}}
]
